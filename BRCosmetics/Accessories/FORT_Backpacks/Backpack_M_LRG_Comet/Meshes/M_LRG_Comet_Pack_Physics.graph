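# Physics asset for the Comet large backpack cosmetic, simulating secondary motion of its dangling pouch, two swords and six rocks.
# Each part has a box body, with small box bodies on base bones acting as anchors.
# Constraints tie each sword and the pouch to its own base and all six rocks to a shared rock base.
# Swing and twist limits are tight, tightest on the rocks, so parts only jiggle slightly.
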PhysicsAsset /BRCosmetics/Accessories/FORT_Backpacks/Backpack_M_LRG_Comet/Meshes/M_LRG_Comet_Pack_Physics
bodies {
 dyn_pouch_base [{shapeType=Box, extentX=0.625, extentY=0.625, extentZ=0.625}]
 dyn_pouch_parent [{shapeType=Box, extentX=11.041, extentY=13.7975, extentZ=4.3825, center=<5.1983 0 -1.584>}]
 dyn_rock_01 [{shapeType=Box, extentX=13.7248, extentY=8.3967, extentZ=9.1022, center=<8.1635 0 0>}]
 dyn_rock_02 [{shapeType=Box, extentX=9.3255, extentY=7.1011, extentZ=7.0846, center=<4.183 0 0>}]
 dyn_rock_03 [{shapeType=Box, extentX=6.9112, extentY=3.5625, extentZ=2.4143, center=<3.1844 0 0>}]
 dyn_rock_04 [{shapeType=Box, extentX=11.2785, extentY=4.3175, extentZ=5.9467, center=<8.0662 0 0>}]
 dyn_rock_05 [{shapeType=Box, extentX=10.8727, extentY=7.7017, extentZ=7.9419, center=<5.1293 0 0>}]
 dyn_rock_06 [{shapeType=Box, extentX=15.4591, extentY=10.3671, extentZ=10.3386, center=<7.4816 0 0>}]
 dyn_rock_base [{shapeType=Box, extentX=1.25, extentY=1.25, extentZ=1.25, center=<-2.9359 0 0>}]
 dyn_sword_base_l [{shapeType=Box, extentX=0.5, extentY=0.5, extentZ=0.5}]
 dyn_sword_base_r [{shapeType=Box, extentX=0.5, extentY=0.5, extentZ=0.5}]
 dyn_sword_l [{shapeType=Box, extentX=31.1169, extentY=8.3374, extentZ=4.9413, center=<5.8559 -0.0002 -1.8408>}]
 dyn_sword_r [{shapeType=Box, extentX=31.1758, extentY=8.1718, extentZ=4.9412, center=<7.1427 -0.0004 -2.1313>}]
}
constraints [
 {bone1Name=dyn_sword_l, bone2Name=dyn_sword_base_l, swing1LimitDegrees=10, swing2LimitDegrees=10, twistLimitDegrees=2}
 {bone1Name=dyn_sword_r, bone2Name=dyn_sword_base_r, swing1LimitDegrees=8, swing2LimitDegrees=8, twistLimitDegrees=2}
 {bone1Name=dyn_pouch_parent, bone2Name=dyn_pouch_base, swing1LimitDegrees=10, swing2LimitDegrees=10, twistLimitDegrees=2}
 {bone1Name=dyn_rock_06, bone2Name=dyn_rock_base, swing1LimitDegrees=3, swing2LimitDegrees=3, twistLimitDegrees=2}
 {bone1Name=dyn_rock_01, bone2Name=dyn_rock_base, swing1LimitDegrees=3, swing2LimitDegrees=3, twistLimitDegrees=2}
 {bone1Name=dyn_rock_02, bone2Name=dyn_rock_base, swing1LimitDegrees=3, swing2LimitDegrees=3, twistLimitDegrees=2}
 {bone1Name=dyn_rock_03, bone2Name=dyn_rock_base, swing1LimitDegrees=3, swing2LimitDegrees=3, twistLimitDegrees=2}
 {bone1Name=dyn_rock_04, bone2Name=dyn_rock_base, swing1LimitDegrees=3, swing2LimitDegrees=3, twistLimitDegrees=2}
 {bone1Name=dyn_rock_05, bone2Name=dyn_rock_base, swing1LimitDegrees=3, swing2LimitDegrees=3, twistLimitDegrees=2}
]
